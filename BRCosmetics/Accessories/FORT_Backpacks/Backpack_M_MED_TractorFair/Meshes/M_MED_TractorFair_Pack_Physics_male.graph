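# Physics asset for the male TractorFair backpack cosmetic, driving secondary motion on its spinal cord and wire.
# Two five-link capsule chains hang from base bodies with no twist, and the spinal cord's first swing limit widens toward the tip.
# Large capsule colliders along the pelvis and spine bones, joined by free-rotation constraints, approximate the wearer's torso for the dangling parts.
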PhysicsAsset /BRCosmetics/Accessories/FORT_Backpacks/Backpack_M_MED_TractorFair/Meshes/M_MED_TractorFair_Pack_Physics_male
bodies {
 base_spinalCord_1 [{shapeType=Capsule, length=2, radius=1, shapeName=base_spinalCord_1_capsule}]
 base_wire_C_1 [{shapeType=Capsule, length=1, radius=0.5, shapeName=base_wire_C_1_capsule}]
 dyn_spinalCord_1 [{shapeType=Capsule, length=0.4, radius=3.4435, shapeName=dyn_spinalCord_1_capsule, center=<2.4145 0 0>, rotation=<rot 0 0 90>}]
 dyn_spinalCord_2 [{shapeType=Capsule, length=0.6373, radius=3.4153, shapeName=dyn_spinalCord_1_capsule, center=<2.4145 0 0.6887>, rotation=<rot 0 0 90>}]
 dyn_spinalCord_3 [
  {shapeType=Capsule, length=0.746, radius=3.6576, shapeName=dyn_spinalCord_1_capsule, center=<2.4145 0 0.5201>, rotation=<rot 0 0 90>}
  {
   shapeType=Capsule
   length=7.508
   radius=8.051
   shapeName=dyn_spinalCord_3_capsule_A
   center=<8.1631 3.2899 11.9879>
   rotation=<rot 0 0 176.1805>
  }
 ]
 dyn_spinalCord_4 [{shapeType=Capsule, length=0.974, radius=3.1934, shapeName=dyn_spinalCord_1_capsule, center=<2.4145 0 0.3745>, rotation=<rot 0 0 90>}]
 dyn_spinalCord_5 [{shapeType=Capsule, length=7.0245, radius=4.1406, shapeName=dyn_spinalCord_1_capsule, center=<6.0088 0 1.433>, rotation=<rot 0 90 92.1892>}]
 dyn_wire_C_1 [{shapeType=Capsule, length=3.1829, radius=0.6981, shapeName=dyn_wire_C_1_capsule, center=<2.5147 0.137 0>, rotation=<rot -90 0 0>}]
 dyn_wire_C_2 [{shapeType=Capsule, length=3.4773, radius=0.8356, shapeName=dyn_wire_C_2_capsule, center=<2.0476 0.0561 0>, rotation=<rot 90 0 0>}]
 dyn_wire_C_3 [{shapeType=Capsule, length=3.0857, radius=0.7457, shapeName=dyn_wire_C_3_capsule, center=<2.2989 -0.0191 -0.1544>, rotation=<rot 90 0 0>}]
 dyn_wire_C_4 [{shapeType=Capsule, length=2.3654, radius=0.8295, shapeName=dyn_wire_C_4_capsule, center=<1.9137 -0.1044 -0.0018>, rotation=<rot 90 0 0>}]
 dyn_wire_C_5 [{shapeType=Capsule, length=1.5415, radius=0.9566, shapeName=dyn_wire_C_5_capsule, center=<1.6391 0 -0.1266>, rotation=<rot 90 0 0>}]
 pelvis_collider [{shapeType=Capsule, length=8.6456, radius=15.1989, shapeName=pelvis_collider_capsule, center=<-2.5051 -0.2706 0>}]
 spine_01_collider [{shapeType=Capsule, length=7.9437, radius=14.0052, shapeName=spine_01_collider_capsule, center=<-1.8231 1.0147 0>}]
 spine_02_collider [{shapeType=Capsule, length=7.3264, radius=13.6322, shapeName=spine_02_collider_capsule, center=<-1.8105 1.0597 0.0001>}]
 spine_03_collider [{shapeType=Capsule, length=8.2107, radius=13.7031, shapeName=spine_03_collider_capsule, center=<-0.1524 2.939 0.0001>}]
 spine_04_collider [{shapeType=Capsule, length=9.4404, radius=14.1325, shapeName=spine_04_collider_capsule, center=<2.766 0.7765 0.0002>}]
 spine_05_collider [{shapeType=Capsule, length=8.0226, radius=15.695, shapeName=spine_05_collider_capsule, center=<-2.2795 1.8109 0>}]
}
constraints [
 {bone1Name=dyn_spinalCord_1, bone2Name=base_spinalCord_1, swing1LimitDegrees=4, swing2LimitDegrees=15, twistLimitDegrees=0}
 {bone1Name=dyn_spinalCord_2, bone2Name=dyn_spinalCord_1, swing1LimitDegrees=6, swing2LimitDegrees=15, twistLimitDegrees=0}
 {bone1Name=dyn_spinalCord_3, bone2Name=dyn_spinalCord_2, swing1LimitDegrees=8, swing2LimitDegrees=15, twistLimitDegrees=0}
 {bone1Name=dyn_spinalCord_4, bone2Name=dyn_spinalCord_3, swing1LimitDegrees=10, swing2LimitDegrees=15, twistLimitDegrees=0}
 {bone1Name=dyn_spinalCord_5, bone2Name=dyn_spinalCord_4, swing1LimitDegrees=12, swing2LimitDegrees=15, twistLimitDegrees=0}
 {
  bone1Name=spine_01_collider
  bone2Name=pelvis_collider
  swing1LimitDegrees=45
  swing1Motion=Free
  swing2LimitDegrees=45
  swing2Motion=Free
  twistLimitDegrees=45
  twistMotion=Free
 }
 {
  bone1Name=spine_02_collider
  bone2Name=spine_01_collider
  swing1LimitDegrees=45
  swing1Motion=Free
  swing2LimitDegrees=45
  swing2Motion=Free
  twistLimitDegrees=45
  twistMotion=Free
 }
 {
  bone1Name=spine_04_collider
  bone2Name=spine_03_collider
  swing1LimitDegrees=45
  swing1Motion=Free
  swing2LimitDegrees=45
  swing2Motion=Free
  twistLimitDegrees=45
  twistMotion=Free
 }
 {
  bone1Name=spine_05_collider
  bone2Name=spine_04_collider
  swing1LimitDegrees=45
  swing1Motion=Free
  swing2LimitDegrees=45
  swing2Motion=Free
  twistLimitDegrees=45
  twistMotion=Free
 }
 {
  bone1Name=spine_03_collider
  bone2Name=spine_02_collider
  swing1LimitDegrees=45
  swing1Motion=Free
  swing2LimitDegrees=45
  swing2Motion=Free
  twistLimitDegrees=45
  twistMotion=Free
 }
 {bone1Name=dyn_wire_C_1, bone2Name=base_wire_C_1, swing1LimitDegrees=20, swing2LimitDegrees=20, twistLimitDegrees=0}
 {bone1Name=dyn_wire_C_2, bone2Name=dyn_wire_C_1, swing1LimitDegrees=15, swing2LimitDegrees=15, twistLimitDegrees=0}
 {bone1Name=dyn_wire_C_3, bone2Name=dyn_wire_C_2, swing1LimitDegrees=15, swing2LimitDegrees=15, twistLimitDegrees=0}
 {bone1Name=dyn_wire_C_4, bone2Name=dyn_wire_C_3, swing1LimitDegrees=15, swing2LimitDegrees=15, twistLimitDegrees=0}
 {bone1Name=dyn_wire_C_5, bone2Name=dyn_wire_C_4, swing1LimitDegrees=15, swing2LimitDegrees=15, twistLimitDegrees=0}
]
