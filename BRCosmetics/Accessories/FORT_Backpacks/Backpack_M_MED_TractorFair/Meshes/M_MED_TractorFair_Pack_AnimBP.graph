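# Post-physics AnimBP for the TractorFair backpack cosmetic, animating its dangling wires and spinal cord on the wearer.
# Each update copies parent pelvis and spine sockets onto collider bones and offsets the pack by gender, spine bend and tail.
# AnimDynamics simulates wires, then gender-specific RigidBody assets run with overridden gravity.
# For tailed characters, additive bone rotations adjust the spinal cord during skydiving, sprint-swimming and parachuting, with a timed stronger pose on chute open.
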AnimBlueprint /BRCosmetics/Accessories/FORT_Backpacks/Backpack_M_MED_TractorFair/Meshes/M_MED_TractorFair_Pack_AnimBP
target_skeleton=&/M_MED_TractorFair_Pack_Skeleton
graphs [
 AnimGraph {nodes [
   Output Pose(Result=Use cached pose 'postSimTweaksCP'()) {layerGroup=None}
   Save cached pose 'inputPoseCP'(Pose=Input Pose()) {cacheName=inputPoseCP}
   n1 = Get isParachutingExtremeWearingTail() -> (isParachutingExtremeWearingTail:double)
   n2 = Get isParachutingAndWearingTail() -> (isParachutingAndWearingTail:double)
   n3 = Get triggerToFixTail 1() -> (triggerToFixTail 1:bool)
   n4 = "Transform (Modify) Bone - Bone: dyn_spinalCord_2"!(
    ComponentPose=Local To Component(LocalPose=Use cached pose 'rigidBodiesCP'())
    bAlphaBoolEnabled:bool=@n3
    Translation="0,0,0"
    Rotation="6,0,0"
   ) -> (Pose) {
    alphaInputType=Bool
    rotationMode=BMM_Additive
    rotationSpace=BCS_BoneSpace
    translationMode=BMM_Additive
    translationSpace=BCS_BoneSpace
    alphaBoolBlend {blendInTime=0.2, blendOption=Cubic, blendOutTime=0.2}
    boneToModify {boneName=dyn_spinalCord_2}
    rotation {pitch=6}
   }
   n5 = "Transform (Modify) Bone - Bone: dyn_spinalCord_3"!(ComponentPose=@n4, bAlphaBoolEnabled:bool=@n3, Translation="0,0,0", Rotation="12,0,0") -> (Pose) {
    alphaInputType=Bool
    rotationMode=BMM_Additive
    rotationSpace=BCS_BoneSpace
    translationMode=BMM_Additive
    translationSpace=BCS_BoneSpace
    alphaBoolBlend {blendInTime=0.2, blendOption=Cubic, blendOutTime=0.2}
    boneToModify {boneName=dyn_spinalCord_3}
    rotation {pitch=12}
   }
   n6 = "Transform (Modify) Bone - Bone: dyn_spinalCord_4"!(ComponentPose=@n5, bAlphaBoolEnabled:bool=@n3, Translation="0,0,0", Rotation="12,0,0") -> (Pose) {
    alphaInputType=Bool
    rotationMode=BMM_Additive
    rotationSpace=BCS_BoneSpace
    translationMode=BMM_Additive
    translationSpace=BCS_BoneSpace
    alphaBoolBlend {blendInTime=0.2, blendOption=Cubic, blendOutTime=0.2}
    boneToModify {boneName=dyn_spinalCord_4}
    rotation {pitch=12}
   }
   n7 = "Transform (Modify) Bone - Bone: dyn_spinalCord_1"!(ComponentPose=@n6, Alpha=@n2, Translation="0,0,0", Rotation="12,0,0") -> (Pose) {
    rotationMode=BMM_Additive
    rotationSpace=BCS_BoneSpace
    translationMode=BMM_Additive
    translationSpace=BCS_BoneSpace
    alphaBoolBlend {blendOption=Linear}
    alphaScaleBiasClamp {bInterpResult=true, interpSpeedDecreasing=3, interpSpeedIncreasing=5}
    boneToModify {boneName=dyn_spinalCord_1}
    rotation {pitch=12}
   }
   n8 = "Transform (Modify) Bone - Bone: dyn_spinalCord_2"!(ComponentPose=@n7, Alpha:float=@n2, Translation="0,0,0", Rotation="12,0,0") -> (Pose) {
    rotationMode=BMM_Additive
    rotationSpace=BCS_BoneSpace
    translationMode=BMM_Additive
    translationSpace=BCS_BoneSpace
    alphaBoolBlend {blendOption=Linear}
    alphaScaleBiasClamp {bInterpResult=true, interpSpeedDecreasing=3, interpSpeedIncreasing=5}
    boneToModify {boneName=dyn_spinalCord_2}
    rotation {pitch=12}
   }
   n9 = "Transform (Modify) Bone - Bone: dyn_spinalCord_3"!(ComponentPose=@n8, Alpha:float=@n2, Translation="0,0,0", Rotation="10,0,0") -> (Pose) {
    rotationMode=BMM_Additive
    rotationSpace=BCS_BoneSpace
    translationMode=BMM_Additive
    translationSpace=BCS_BoneSpace
    alphaBoolBlend {blendOption=Linear}
    alphaScaleBiasClamp {bInterpResult=true, interpSpeedDecreasing=3, interpSpeedIncreasing=5}
    boneToModify {boneName=dyn_spinalCord_3}
    rotation {pitch=10}
   }
   n10 = "Transform (Modify) Bone - Bone: dyn_spinalCord_4"!(ComponentPose=@n9, Alpha:float=@n2, Translation="0,0,0", Rotation="8,0,0") -> (Pose) {
    rotationMode=BMM_Additive
    rotationSpace=BCS_BoneSpace
    translationMode=BMM_Additive
    translationSpace=BCS_BoneSpace
    alphaBoolBlend {blendOption=Linear}
    alphaScaleBiasClamp {bInterpResult=true, interpSpeedDecreasing=3, interpSpeedIncreasing=5}
    boneToModify {boneName=dyn_spinalCord_4}
    rotation {pitch=8}
   }
   n11 = "Transform (Modify) Bone - Bone: pack"!(ComponentPose=@n10, Alpha=@n1, Translation="0,0,2", Rotation="0,0,12") -> (Pose) {
    rotationMode=BMM_Additive
    rotationSpace=BCS_BoneSpace
    translationMode=BMM_Additive
    translationSpace=BCS_BoneSpace
    alphaBoolBlend {blendOption=Linear}
    alphaScaleBiasClamp {bInterpResult=true, interpSpeedDecreasing=2, interpSpeedIncreasing=5}
    boneToModify {boneName=pack}
    rotation {roll=12}
    translation {z=2}
   }
   n12 = "Transform (Modify) Bone - Bone: dyn_spinalCord_1"(ComponentPose=@n11, Alpha=@n1, Translation="0,0,0", Rotation="10,0,0") {
    rotationMode=BMM_Additive
    rotationSpace=BCS_BoneSpace
    translationMode=BMM_Additive
    translationSpace=BCS_BoneSpace
    alphaBoolBlend {blendOption=Linear}
    alphaScaleBiasClamp {bInterpResult=true, interpSpeedDecreasing=2, interpSpeedIncreasing=5}
    boneToModify {boneName=dyn_spinalCord_1}
    rotation {pitch=10}
   }
   n13 = "Transform (Modify) Bone - Bone: dyn_spinalCord_2"(ComponentPose=@n12, Alpha=@n1, Translation="0,0,0", Rotation="8,0,0") {
    rotationMode=BMM_Additive
    rotationSpace=BCS_BoneSpace
    translationMode=BMM_Additive
    translationSpace=BCS_BoneSpace
    alphaBoolBlend {blendOption=Linear}
    alphaScaleBiasClamp {bInterpResult=true, interpSpeedDecreasing=2, interpSpeedIncreasing=5}
    boneToModify {boneName=dyn_spinalCord_2}
    rotation {pitch=8}
   }
   Save cached pose 'postSimTweaksCP'(Pose=Component To Local(ComponentPose=@n13)) {cacheName=postSimTweaksCP}
   n14 = RigidBody!(
    ComponentPose=Local To Component(LocalPose=Use cached pose 'animDynCP'())
    Alpha
    OverrideWorldGravity:Vector=Get GravityOverride()
    SimSpaceSettings:Sim Space Settings Structure=Get Sim Space Settings() -> (Sim Space Settings:Sim Space Settings Structure)
   ) -> (Pose) {
    bOverrideWorldGravity=true
    overridePhysicsAsset=&/M_MED_TractorFair_Pack_Physics_female2
    simulationSpace=BaseBoneSpace
    alphaBoolBlend {blendOption=Linear}
    baseBoneRef {boneName=root}
    componentAppliedLinearAccClamp=<3000 3000 3000>
    componentLinearAccScale=<0.4 0.3 0.4>
    componentLinearVelScale=<0.4 0.3 0.4>
    overrideWorldGravity {z=-700}
    simSpaceSettings {maxAngularVelocity=3, maxLinearAcceleration=700, maxLinearVelocity=500, worldAlpha=0.3, externalLinearDragV=<0.2 0.2 0.2>}
   }
   n15 = RigidBody(
    ComponentPose=Local To Component(LocalPose=Use cached pose 'animDynCP'())
    OverrideWorldGravity=Get GravityOverride()
    SimSpaceSettings=Get Sim Space Settings()
   ) {
    bOverrideWorldGravity=true
    overridePhysicsAsset=&/M_MED_TractorFair_Pack_Physics_male
    simulationSpace=BaseBoneSpace
    alphaBoolBlend {blendOption=Linear}
    baseBoneRef {boneName=root}
    componentAppliedLinearAccClamp=<3000 3000 3000>
    componentLinearAccScale=<0.4 0.3 0.4>
    componentLinearVelScale=<0.4 0.3 0.4>
    overrideWorldGravity {z=-700}
    simSpaceSettings {maxAngularVelocity=3, maxLinearAcceleration=700, maxLinearVelocity=500, worldAlpha=0.3, externalLinearDragV=<0.2 0.2 0.2>}
   }
   Save cached pose 'rigidBodiesCP'(Pose=Blend Poses by bool(
     BlendPose_0=Component To Local(ComponentPose=@n15)
     BlendPose_1=Component To Local(ComponentPose=@n14)
     bActiveValue=Get isMale() -> (isMale:bool)
    ) {blendType=Linear}) {cacheName=rigidBodiesCP}
   n16 = Get GravityOverride()
   n17 = Get GravityOverride()
   n18 = Anim Dynamics - Bone: dyn_wire_A_1!(ComponentPose=Local To Component(LocalPose=Use cached pose 'preSimTweaksCP'()), Alpha=1, GravityOverride:Vector=@n17) -> (Pose) {
    angularDampingOverride=1
    angularSpringConstant=4000
    bAngularSpring=true
    bOverrideAngularDamping=true
    bUseGravityOverride=true
    gravityScale=0.35
    simulationSpace=RootRelative
    alphaBoolBlend {blendOption=Linear}
    boundBone {boneName=dyn_wire_A_1}
    componentAppliedLinearAccClamp=<3000 3000 3000>
    componentLinearAccScale=<0.4 0.4 0.4>
    componentLinearVelScale=<0.4 0.4 0.4>
    gravityOverride {z=-980}
    physicsBodyDefinitions [{
      boundBone {boneName=dyn_wire_A_1}
      boxExtents=<30 30 30>
      constraintSetup {angularLimitsMax=<0 25 10>, angularLimitsMin=<0 -15 -10>, angularTarget=<1 0 0>}
      localJointOffset=<6 0 0>
     }]
   }
   n19 = Anim Dynamics - Bone: dyn_wire_A_2!(ComponentPose=@n18, Alpha=1, GravityOverride:Vector=@n17) -> (Pose) {
    angularDampingOverride=1
    angularSpringConstant=4000
    bAngularSpring=true
    bOverrideAngularDamping=true
    bUseGravityOverride=true
    gravityScale=0.35
    simulationSpace=RootRelative
    alphaBoolBlend {blendOption=Linear}
    boundBone {boneName=dyn_wire_A_2}
    componentAppliedLinearAccClamp=<3000 3000 3000>
    componentLinearAccScale=<0.4 0.4 0.4>
    componentLinearVelScale=<0.4 0.4 0.4>
    gravityOverride {z=-980}
    physicsBodyDefinitions [{
      boundBone {boneName=dyn_wire_A_2}
      boxExtents=<30 30 30>
      constraintSetup {angularLimitsMax=<0 20 10>, angularLimitsMin=<0 -15 -10>, angularTarget=<1 0 0>}
      localJointOffset=<8 0 0>
     }]
   }
   n20 = Anim Dynamics - Bone: dyn_wire_A_3!(ComponentPose=@n19, Alpha=1, GravityOverride:Vector=@n17) -> (Pose) {
    angularDampingOverride=1
    angularSpringConstant=4000
    bAngularSpring=true
    bOverrideAngularDamping=true
    bUseGravityOverride=true
    gravityScale=0.35
    simulationSpace=RootRelative
    alphaBoolBlend {blendOption=Linear}
    boundBone {boneName=dyn_wire_A_3}
    componentAppliedLinearAccClamp=<3000 3000 3000>
    componentLinearAccScale=<0.6 0.6 0.6>
    componentLinearVelScale=<0.6 0.6 0.6>
    gravityOverride {z=-980}
    physicsBodyDefinitions [{
      boundBone {boneName=dyn_wire_A_3}
      boxExtents=<30 30 30>
      constraintSetup {angularLimitsMax=<0 15 10>, angularLimitsMin=<0 -15 -10>, angularTarget=<1 0 0>}
      localJointOffset=<10 0 0>
     }]
   }
   n21 = Anim Dynamics - Bone: dyn_wire_A_4!(ComponentPose=@n20, Alpha=1, GravityOverride:Vector=@n17) -> (Pose) {
    angularDampingOverride=1
    angularSpringConstant=4000
    bAngularSpring=true
    bOverrideAngularDamping=true
    bUseGravityOverride=true
    gravityScale=0.35
    simulationSpace=RootRelative
    alphaBoolBlend {blendOption=Linear}
    boundBone {boneName=dyn_wire_A_4}
    componentAppliedLinearAccClamp=<3000 3000 3000>
    componentLinearAccScale=<0.6 0.6 0.6>
    componentLinearVelScale=<0.6 0.6 0.6>
    gravityOverride {z=-980}
    physicsBodyDefinitions [{
      boundBone {boneName=dyn_wire_A_4}
      boxExtents=<30 30 30>
      constraintSetup {angularLimitsMax=<0 20 10>, angularLimitsMin=<0 -20 -10>, angularTarget=<1 0 0>}
      localJointOffset=<12 0 0>
     }]
   }
   n22 = Anim Dynamics - Bone: dyn_wire_B_1!(ComponentPose=@n21, Alpha=1, GravityOverride:Vector=@n16) -> (Pose) {
    angularDampingOverride=1
    angularSpringConstant=4000
    bAngularSpring=true
    bOverrideAngularDamping=true
    bUseGravityOverride=true
    gravityScale=0.35
    simulationSpace=RootRelative
    alphaBoolBlend {blendOption=Linear}
    boundBone {boneName=dyn_wire_B_1}
    componentAppliedLinearAccClamp=<3000 3000 3000>
    componentLinearAccScale=<0.4 0.4 0.4>
    componentLinearVelScale=<0.4 0.4 0.4>
    gravityOverride {z=-980}
    physicsBodyDefinitions [{
      boundBone {boneName=dyn_wire_B_1}
      boxExtents=<30 30 30>
      constraintSetup {angularLimitsMax=<0 20 10>, angularLimitsMin=<0 -10 -15>, angularTarget=<1 0 0>}
      localJointOffset=<2 0 0>
     }]
   }
   n23 = Anim Dynamics - Bone: dyn_wire_B_2!(ComponentPose=@n22, Alpha=1, GravityOverride:Vector=@n16) -> (Pose) {
    angularDampingOverride=1
    angularSpringConstant=4000
    bAngularSpring=true
    bOverrideAngularDamping=true
    bUseGravityOverride=true
    gravityScale=0.35
    simulationSpace=RootRelative
    alphaBoolBlend {blendOption=Linear}
    boundBone {boneName=dyn_wire_B_2}
    componentAppliedLinearAccClamp=<3000 3000 3000>
    componentLinearAccScale=<0.4 0.4 0.4>
    componentLinearVelScale=<0.4 0.4 0.4>
    gravityOverride {z=-980}
    physicsBodyDefinitions [{
      boundBone {boneName=dyn_wire_B_2}
      boxExtents=<30 30 30>
      constraintSetup {angularLimitsMax=<0 20 12>, angularLimitsMin=<0 -10 -15>, angularTarget=<1 0 0>}
      localJointOffset=<4 0 0>
     }]
   }
   Save cached pose 'animDynCP'(Pose=Component To Local(ComponentPose=@n23)) {cacheName=animDynCP}
   n24 = "Transform (Modify) Bone - Bone: pelvis_collider"(
    ComponentPose=Local To Component(LocalPose=Use cached pose 'inputPoseCP'())
    Alpha=1
    Translation=Get pelvis_trans() -> (pelvis_trans:Vector)
    Rotation=Get pelvis_rot() -> (pelvis_rot:Rotator)
    Scale="1,1,1"
   ) {
    rotationMode=BMM_Replace
    rotationSpace=BCS_WorldSpace
    scaleSpace=BCS_BoneSpace
    translationMode=BMM_Replace
    translationSpace=BCS_WorldSpace
    alphaBoolBlend {blendOption=Linear}
    alphaScaleBiasClamp {inRange {max=0, min=-40}, outRange {max=0, min=1}}
    boneToModify {boneName=pelvis_collider}
   }
   n25 = "Transform (Modify) Bone - Bone: spine_01_collider"(
    ComponentPose=@n24
    Alpha=1
    Translation=Get spine_01_trans() -> (spine_01_trans:Vector)
    Rotation=Get spine_01_rot() -> (spine_01_rot:Rotator)
    Scale="1,1,1"
   ) {
    rotationMode=BMM_Replace
    rotationSpace=BCS_WorldSpace
    scaleSpace=BCS_BoneSpace
    translationMode=BMM_Replace
    translationSpace=BCS_WorldSpace
    alphaBoolBlend {blendOption=Linear}
    alphaScaleBiasClamp {inRange {max=0, min=-40}, outRange {max=0, min=1}}
    boneToModify {boneName=spine_01_collider}
   }
   n26 = "Transform (Modify) Bone - Bone: spine_02_collider"(
    ComponentPose=@n25
    Alpha=1
    Translation=Get spine_02_trans() -> (spine_02_trans:Vector)
    Rotation=Get spine_02_rot() -> (spine_02_rot:Rotator)
    Scale="1,1,1"
   ) {
    rotationMode=BMM_Replace
    rotationSpace=BCS_WorldSpace
    scaleSpace=BCS_BoneSpace
    translationMode=BMM_Replace
    translationSpace=BCS_WorldSpace
    alphaBoolBlend {blendOption=Linear}
    alphaScaleBiasClamp {inRange {max=0, min=-40}, outRange {max=0, min=1}}
    boneToModify {boneName=spine_02_collider}
   }
   n27 = "Transform (Modify) Bone - Bone: spine_03_collider"(
    ComponentPose=@n26
    Alpha=1
    Translation=Get spine_03_trans() -> (spine_03_trans:Vector)
    Rotation=Get spine_03_rot() -> (spine_03_rot:Rotator)
    Scale="1,1,1"
   ) {
    rotationMode=BMM_Replace
    rotationSpace=BCS_WorldSpace
    scaleSpace=BCS_BoneSpace
    translationMode=BMM_Replace
    translationSpace=BCS_WorldSpace
    alphaBoolBlend {blendOption=Linear}
    alphaScaleBiasClamp {inRange {max=0, min=-40}, outRange {max=0, min=1}}
    boneToModify {boneName=spine_03_collider}
   }
   n28 = "Transform (Modify) Bone - Bone: spine_04_collider"(
    ComponentPose=@n27
    Alpha=1
    Translation=Get spine_04_trans() -> (spine_04_trans:Vector)
    Rotation=Get spine_04_rot() -> (spine_04_rot:Rotator)
    Scale="1,1,1"
   ) {
    rotationMode=BMM_Replace
    rotationSpace=BCS_WorldSpace
    scaleSpace=BCS_BoneSpace
    translationMode=BMM_Replace
    translationSpace=BCS_WorldSpace
    alphaBoolBlend {blendOption=Linear}
    alphaScaleBiasClamp {inRange {max=0, min=-40}, outRange {max=0, min=1}}
    boneToModify {boneName=spine_04_collider}
   }
   n29 = "Transform (Modify) Bone - Bone: spine_05_collider"(
    ComponentPose=@n28
    Alpha=1
    Translation=Get spine_05_trans() -> (spine_05_trans:Vector)
    Rotation=Get spine_05_rot() -> (spine_05_rot:Rotator)
    Scale="1,1,1"
   ) {
    rotationMode=BMM_Replace
    rotationSpace=BCS_WorldSpace
    scaleSpace=BCS_BoneSpace
    translationMode=BMM_Replace
    translationSpace=BCS_WorldSpace
    alphaBoolBlend {blendOption=Linear}
    alphaScaleBiasClamp {inRange {max=0, min=-40}, outRange {max=0, min=1}}
    boneToModify {boneName=spine_05_collider}
   }
   Save cached pose 'spineMatchedCP'(Pose=Component To Local(ComponentPose=@n29)) {cacheName=spineMatchedCP}
   n30 = "Transform (Modify) Bone - Bone: pack"!(
    ComponentPose=Local To Component(LocalPose=Use cached pose 'spineMatchedCP'())
    bAlphaBoolEnabled:bool
    Translation=Get packOffset() -> (packOffset:Vector)
   ) -> (Pose) {
    alphaInputType=Bool
    translationMode=BMM_Additive
    translationSpace=BCS_BoneSpace
    alphaBoolBlend {blendOption=Linear}
    boneToModify {boneName=pack}
   }
   n31 = "Transform (Modify) Bone - Bone: pack"!(ComponentPose=@n30, Alpha:float, Translation=Get backpack_translation(), Rotation:Rotator=Get backpack_rotate()) -> (Pose) {rotationMode=BMM_Additive, translationMode=BMM_Additive, alphaBoolBlend {blendOption=Linear}, boneToModify {boneName=pack}}
   n32 = "Transform (Modify) Bone - Bone: pack"!(ComponentPose=@n31, bAlphaBoolEnabled:bool=Get isCharacterWithTail() -> (isCharacterWithTail:bool), Translation="0,0,3", Rotation="0,0,0") -> (Pose) {
    alphaInputType=Bool
    rotationMode=BMM_Additive
    rotationSpace=BCS_BoneSpace
    translationMode=BMM_Additive
    translationSpace=BCS_BoneSpace
    alphaBoolBlend {blendOption=Linear}
    boneToModify {boneName=pack}
    translation {z=3}
   }
   n33 = "Constraint - base_wire_C_1 (P)"(ComponentPose=@n32, Alpha=1, ConstraintWeights_0=1) {
    alphaBoolBlend {blendOption=Linear}
    boneToModify {boneName=base_wire_C_1}
    constraintSetup [{targetBone {boneName=dyn_spinalCord_1}}]
    constraintWeights [1]
   }
   Save cached pose 'preSimTweaksCP'(Pose=Component To Local(ComponentPose=@n33)) {cacheName=preSimTweaksCP}
  ]}
 EventGraph {nodes [
   Event BlueprintUpdateAnimation()
   > n1 = Sequence!(execute) -> (then_0, then_1, then_2:Exec, then_3:Exec)
   > n2 = Gravity Override State Function DEPRECATED(CharacterPartAnimInstance=Self-Reference(), GameTimeInSeconds=GetGameTimeInSeconds(), GravityOverrideParams=Get Gravity Override Params())
   > Set GravityOverride(GravityOverride=@n2)
   n3 = GetOwnerSkeletalMeshComponent()
   n4 = BreakRotator(InRot=BreakTransform(InTransform=GetDeltaTransformFromRefPose(self=@n3, BoneName=spine_05, BaseName=spine_01)).Rotation)
   Set backpack_translation(execute=@n1.then_1, backpack_translation=MakeVector(Y=MapRangeClamped(Value=@n4.Yaw, InRangeA=0, InRangeB=-30, OutRangeB=-2)))
   > Set backpack_rotate(backpack_rotate=MakeRotator(Roll=MapRangeClamped(Value=@n4.Yaw, InRangeA=0, InRangeB=-30, OutRangeB=10)))
   n5 = GetPartSkeletalMeshComponent()
   Is Valid(exec=@n1.then_2, InputObject=@n5)
   n6 = GetAttachParent!(self:Scene Component Object Reference=@n5) -> (ReturnValue:Scene Component Object Reference)
   > Is Valid(InputObject=@n6)
   n7 = BreakTransform(InTransform=GetSocketTransform(self=@n6, InSocketName=spine_01, TransformSpace=RTS_World))
   > Set spine_01_trans(spine_01_trans:Vector=@n7) -> (Output_Get:Vector)
   > Set spine_01_rot(spine_01_rot:Rotator=@n7.Rotation) -> (Output_Get:Rotator)
   n8 = BreakTransform(InTransform=GetSocketTransform(self=@n6, InSocketName=spine_02, TransformSpace=RTS_World))
   > Set spine_02_trans(spine_02_trans:Vector=@n8) -> (Output_Get:Vector)
   > Set spine_02_rot(spine_02_rot:Rotator=@n8.Rotation) -> (Output_Get:Rotator)
   n9 = BreakTransform(InTransform=GetSocketTransform(self=@n6, InSocketName=spine_03, TransformSpace=RTS_World))
   > Set spine_03_trans(spine_03_trans:Vector=@n9) -> (Output_Get:Vector)
   > Set spine_03_rot(spine_03_rot:Rotator=@n9.Rotation) -> (Output_Get:Rotator)
   n10 = BreakTransform(InTransform=GetSocketTransform(self=@n6, InSocketName=spine_04, TransformSpace=RTS_World))
   > Set spine_04_trans(spine_04_trans:Vector=@n10) -> (Output_Get:Vector)
   > Set spine_04_rot(spine_04_rot:Rotator=@n10.Rotation) -> (Output_Get:Rotator)
   n11 = BreakTransform(InTransform=GetSocketTransform(self=@n6, InSocketName=spine_05, TransformSpace=RTS_World))
   > Set spine_05_trans(spine_05_trans:Vector=@n11) -> (Output_Get:Vector)
   > Set spine_05_rot(spine_05_rot:Rotator=@n11.Rotation) -> (Output_Get:Rotator)
   n12 = BreakTransform(InTransform=GetSocketTransform(self=@n6, InSocketName=pelvis, TransformSpace=RTS_World))
   > Set pelvis_trans(pelvis_trans:Vector=@n12) -> (Output_Get:Vector)
   > Set pelvis_rot(pelvis_rot:Rotator=@n12.Rotation) -> (Output_Get:Rotator)
   n13 = Get bIsSkydiveFloating()
   n14 = Get bIsSkydiveDiveMode()
   n15 = Get bIsSwimSprinting() -> (bIsSwimSprinting:bool)
   Set triggerToFixTail 1(
    execute=@n1.then_3
    triggerToFixTail 1:bool=AND Boolean(A=Get isCharacterWithTail(), B=OR Boolean!(A=@n13, B=@n14, C:bool=@n15) -> (ReturnValue))
   ) -> (Output_Get:bool)
   Event OnCharacterCustomizationLoaded()
   > n16 = Sequence!(execute) -> (then_0, then_1)
   > n17 = Set Owning Pawn(Owning Pawn:Pawn Object Reference=TryGetFortPawn()) -> (Output_Get:Pawn Object Reference)
   > Is Valid(InputObject=@n17)
   > n18 = FindGender!(execute, Object:Object Reference=Get Owning Pawn() -> (Owning Pawn:Pawn Object Reference)) -> (True:Exec, False:Exec, OutGender:EFortCustomGender Enum=Invalid)
   > n19 = Set isMale(isMale:bool="Equal (Enum)"(A:EFortCustomGender Enum=@n18, B:EFortCustomGender Enum=Male)) -> (Output_Get:bool)
   > n20 = Branch(Condition=@n19)
   > Set packOffset(packOffset:Vector="0,-1.5,0") -> (Output_Get:Vector)
   Set packOffset(execute=@n20.else, packOffset="0,-1,0")
   Set isCharacterWithTail(execute=@n16.then_1, isCharacterWithTail:bool=CheckForTail!(self:Custom Character Part Anim Instance Object Reference) -> (bHasTail:bool)) -> (Output_Get:bool)
   Event OnParachuteClosed()
   > Set isParachutingAndWearingTail(isParachutingAndWearingTail:double=0) -> (Output_Get:double)
   > n21 = Set isParachutingExtremeWearingTail(isParachutingExtremeWearingTail:double=0) -> (Output_Get:double)
   ResetDynamics(execute=[@n22.then, @n21.then], InTeleportType=TeleportPhysics)
   > ResetDynamics(InTeleportType=ResetPhysics)
   n22 = Event OnEndSwimming() {is=title K2Node_Event}
   Event BlueprintInitializeAnimation()
   > Set Skel_Mesh(Skel_Mesh=GetOwnerSkeletalMeshComponent())
   Event OnParachuteOpened()
   > Branch(Condition=Get isCharacterWithTail())
   > Set isParachutingAndWearingTail(isParachutingAndWearingTail=1)
   > Set isParachutingExtremeWearingTail(isParachutingExtremeWearingTail=1)
   > Delay(Duration=2.2)
   > Set isParachutingExtremeWearingTail(isParachutingExtremeWearingTail=0)
  ]}
]
variables [
 backpack_rotate {type=Rotator}
 backpack_translation {type=Vector}
 Gravity Override Params {type=GravityOverrideParamsStruct}
 GravityOverride {type=Vector}
 isCharacterWithTail {type=boolean}
 isMale {type=boolean, default=true}
 isParachutingAndWearingTail {type=number}
 isParachutingExtremeWearingTail {type=number}
 Owning Pawn {type=Pawn, default=null}
 packOffset {type=Vector, default=<0 -1.5 0>}
 pelvis_rot {type=Rotator}
 pelvis_trans {type=Vector}
 Sim Space Settings {
  type=SimSpaceSettings
  default {
   dampingAlpha=1
   maxAngularAcceleration=10000
   maxAngularVelocity=3
   maxLinearAcceleration=700
   maxLinearVelocity=500
   velocityScaleZ=1
   worldAlpha=0.2
   externalLinearDragV=<0.2 0.2 0.2>
  }
 }
 Skel_Mesh {type=SkeletalMeshComponent}
 spine_01_rot {type=Rotator}
 spine_01_trans {type=Vector}
 spine_02_rot {type=Rotator}
 spine_02_trans {type=Vector}
 spine_03_rot {type=Rotator}
 spine_03_trans {type=Vector}
 spine_04_rot {type=Rotator}
 spine_04_trans {type=Vector}
 spine_05_rot {type=Rotator}
 spine_05_trans {type=Vector}
 triggerToFixTail 1 {type=boolean}
]
